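# Animation Blueprint for the FondAim backpack cosmetic, adding springy AnimDynamics motion to its head, ear and arrow bones.
# Two matching chains branch from the cached PreSimPose; bIsInFrontEnd selects the front-end one, which has tighter acceleration clamps and a stiffer head spring.
# The EventGraph maps spine_05 yaw to BackPack Rotation and Translation, but the graph never reads them, and the pack bone modifier has zero offsets.
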AnimBlueprint /BRCosmetics/Accessories/FORT_Backpacks/Backpack_F_MED_FondAim/Meshes/F_MED_FondAim_Pack_AnimBP
target_skeleton=&/F_MED_FondAim_Pack_Skeleton
graphs [
 AnimGraph {nodes [
   n1 = Anim Dynamics - Bone: dyn_head(ComponentPose=Local To Component(LocalPose=Use cached pose 'PreSimPose'()), Alpha=1, ExternalForce="0,0,0") {
    angularDampingOverride=0.9998
    angularSpringConstant=5000
    bAngularSpring=true
    bOverrideAngularDamping=true
    bUseGravityOverride=true
    simulationSpace=RootRelative
    alphaBoolBlend {blendInTime=0.1, blendOption=Linear, blendOutTime=0.1}
    boundBone {boneName=dyn_head}
    componentAppliedLinearAccClamp=<700 700 1000>
    componentLinearAccScale=<0.6 0.6 0.6>
    componentLinearVelScale=<0.8 0.8 0.8>
    gravityOverride {z=-500}
    physicsBodyDefinitions [{
      boundBone {boneName=dyn_head}
      boxExtents=<25 25 25>
      constraintSetup {angularLimitsMax=<2.5 15 5>, angularLimitsMin=<-2.5 -15 -15>, angularTarget=<1 0 0>}
      localJointOffset=<2 0 0>
     }]
   }
   n2 = Anim Dynamics - Bone: dyn_ear_01_l(ComponentPose=@n1, Alpha=1, ExternalForce="0,0,0") {
    angularDampingOverride=0.9998
    angularSpringConstant=5000
    bAngularSpring=true
    bOverrideAngularDamping=true
    bUseGravityOverride=true
    simulationSpace=RootRelative
    alphaBoolBlend {blendInTime=0.1, blendOption=Linear, blendOutTime=0.1}
    boundBone {boneName=dyn_ear_01_l}
    componentAppliedLinearAccClamp=<500 500 1000>
    componentLinearAccScale=<0.6 0.6 0.6>
    componentLinearVelScale=<0.8 0.8 0.8>
    gravityOverride {z=-500}
    physicsBodyDefinitions [{
      boundBone {boneName=dyn_ear_01_l}
      boxExtents=<20 20 20>
      constraintSetup {angularLimitsMax=<2.5 50 15>, angularLimitsMin=<-2.5 -30 -25>, angularTarget=<1 0 0>}
      localJointOffset=<2 0 0>
     }]
   }
   n3 = Anim Dynamics - Bone: dyn_ear_02_l(ComponentPose=@n2, Alpha=1, ExternalForce="0,0,0") {
    angularDampingOverride=0.9999
    angularSpringConstant=4500
    bAngularSpring=true
    bOverrideAngularDamping=true
    bUseGravityOverride=true
    simulationSpace=RootRelative
    alphaBoolBlend {blendInTime=0.1, blendOption=Linear, blendOutTime=0.1}
    boundBone {boneName=dyn_ear_02_l}
    componentAppliedLinearAccClamp=<500 500 1000>
    componentLinearAccScale=<0.6 0.6 0.6>
    componentLinearVelScale=<0.8 0.8 0.8>
    gravityOverride {z=-500}
    physicsBodyDefinitions [{
      boundBone {boneName=dyn_ear_02_l}
      boxExtents=<15 15 15>
      constraintSetup {angularLimitsMax=<2.5 50 15>, angularLimitsMin=<-2.5 -30 -25>, angularTarget=<1 0 0>}
      localJointOffset=<1 0 0>
     }]
   }
   n4 = Anim Dynamics - Bone: dyn_ear_01_r(ComponentPose=@n3, Alpha=1, ExternalForce="0,0,0") {
    angularDampingOverride=0.9998
    angularSpringConstant=5000
    bAngularSpring=true
    bOverrideAngularDamping=true
    bUseGravityOverride=true
    simulationSpace=RootRelative
    alphaBoolBlend {blendInTime=0.1, blendOption=Linear, blendOutTime=0.1}
    boundBone {boneName=dyn_ear_01_r}
    componentAppliedLinearAccClamp=<500 500 1000>
    componentLinearAccScale=<0.6 0.6 0.6>
    componentLinearVelScale=<0.8 0.8 0.8>
    gravityOverride {z=-500}
    physicsBodyDefinitions [{
      boundBone {boneName=dyn_ear_01_r}
      boxExtents=<20 20 20>
      constraintSetup {angularLimitsMax=<2.5 30 15>, angularLimitsMin=<-2.5 -50 -25>, angularTarget=<1 0 0>}
      localJointOffset=<2 0 0>
     }]
   }
   n5 = Anim Dynamics - Bone: dyn_ear_02_r(ComponentPose=@n4, Alpha=1, ExternalForce="0,0,0") {
    angularDampingOverride=0.9999
    angularSpringConstant=4500
    bAngularSpring=true
    bOverrideAngularDamping=true
    bUseGravityOverride=true
    simulationSpace=RootRelative
    alphaBoolBlend {blendInTime=0.1, blendOption=Linear, blendOutTime=0.1}
    boundBone {boneName=dyn_ear_02_r}
    componentAppliedLinearAccClamp=<500 500 1000>
    componentLinearAccScale=<0.6 0.6 0.6>
    componentLinearVelScale=<0.8 0.8 0.8>
    gravityOverride {z=-500}
    physicsBodyDefinitions [{
      boundBone {boneName=dyn_ear_02_r}
      boxExtents=<15 15 15>
      constraintSetup {angularLimitsMax=<2.5 30 15>, angularLimitsMin=<-2.5 -50 -25>, angularTarget=<1 0 0>}
      localJointOffset=<1 0 0>
     }]
   }
   n6 = Anim Dynamics - Bone: dyn_ear_03_r(ComponentPose=@n5, Alpha=1, ExternalForce="0,0,0") {
    angularDampingOverride=0.9999
    angularSpringConstant=2500
    bAngularSpring=true
    bOverrideAngularDamping=true
    bUseGravityOverride=true
    simulationSpace=RootRelative
    alphaBoolBlend {blendInTime=0.1, blendOption=Linear, blendOutTime=0.1}
    boundBone {boneName=dyn_ear_03_r}
    componentAppliedLinearAccClamp=<500 500 1000>
    componentLinearAccScale=<0.6 0.6 0.6>
    componentLinearVelScale=<0.8 0.8 0.8>
    gravityOverride {z=-500}
    physicsBodyDefinitions [{
      boundBone {boneName=dyn_ear_03_r}
      boxExtents=<10 10 10>
      constraintSetup {angularLimitsMax=<2.5 30 10>, angularLimitsMin=<-2.5 -30 -10>, angularTarget=<1 0 0>}
      localJointOffset=<0.5 0 0>
     }]
   }
   n7 = Anim Dynamics - Bone: dyn_arrow_01(ComponentPose=@n6, Alpha=1, ExternalForce="0,0,0") {
    angularDampingOverride=0.999
    angularSpringConstant=7000
    bAngularSpring=true
    bOverrideAngularDamping=true
    bUseGravityOverride=true
    simulationSpace=RootRelative
    boundBone {boneName=dyn_arrow_01}
    componentAppliedLinearAccClamp=<2500 2500 500>
    componentLinearAccScale=<0.5 0.5 0.5>
    componentLinearVelScale=<0.7 0.7 0.7>
    gravityOverride {z=-500}
    physicsBodyDefinitions [{
      boundBone {boneName=dyn_arrow_01}
      boxExtents=<25 25 25>
      constraintSetup {angularLimitsMax=<10 10 10>, angularLimitsMin=<-10 -10 -10>, angularTarget=<1 0 0>}
      localJointOffset=<1 0 0>
     }]
   }
   n8 = Anim Dynamics - Bone: dyn_arrow_02(ComponentPose=@n7, Alpha=1, ExternalForce="0,0,0") {
    angularDampingOverride=0.9999
    angularSpringConstant=3000
    bAngularSpring=true
    bOverrideAngularDamping=true
    bUseGravityOverride=true
    simulationSpace=RootRelative
    boundBone {boneName=dyn_arrow_02}
    componentAppliedLinearAccClamp=<700 700 700>
    componentLinearAccScale=<0.5 0.5 0.5>
    componentLinearVelScale=<0.7 0.7 0.7>
    gravityOverride {z=-750}
    physicsBodyDefinitions [{
      boundBone {boneName=dyn_arrow_02}
      boxExtents=<15 15 15>
      constraintSetup {angularLimitsMax=<10 10 10>, angularLimitsMin=<-10 -10 -10>, angularTarget=<1 0 0>}
      localJointOffset=<0.25 0 -1>
     }]
   }
   n9 = Anim Dynamics - Bone: dyn_arrow_03(ComponentPose=@n8, Alpha=1, ExternalForce="0,0,0") {
    angularDampingOverride=0.997
    angularSpringConstant=5500
    bAngularSpring=true
    bOverrideAngularDamping=true
    bUseGravityOverride=true
    simulationSpace=RootRelative
    boundBone {boneName=dyn_arrow_03}
    componentAppliedLinearAccClamp=<1500 1500 3500>
    componentLinearAccScale=<0.5 0.5 0.5>
    componentLinearVelScale=<0.7 0.7 0.7>
    gravityOverride {z=-800}
    physicsBodyDefinitions [{
      boundBone {boneName=dyn_arrow_03}
      boxExtents=<20 20 20>
      constraintSetup {angularLimitsMax=<10 10 10>, angularLimitsMin=<-10 -10 -10>, angularTarget=<1 0 0>}
      localJointOffset=<0.3 -1 0>
     }]
   }
   n10 = Anim Dynamics - Bone: dyn_head(ComponentPose=Local To Component(LocalPose=Use cached pose 'PreSimPose'()), Alpha=1, ExternalForce="0,0,0") {
    angularDampingOverride=0.9998
    angularSpringConstant=3500
    bAngularSpring=true
    bOverrideAngularDamping=true
    bUseGravityOverride=true
    simulationSpace=RootRelative
    alphaBoolBlend {blendInTime=0.1, blendOption=Linear, blendOutTime=0.1}
    boundBone {boneName=dyn_head}
    componentAppliedLinearAccClamp=<2500 2500 1000>
    componentLinearAccScale=<0.6 0.6 0.6>
    componentLinearVelScale=<0.8 0.8 0.8>
    gravityOverride {z=-500}
    physicsBodyDefinitions [{
      boundBone {boneName=dyn_head}
      boxExtents=<25 25 25>
      constraintSetup {angularLimitsMax=<2.5 15 5>, angularLimitsMin=<-2.5 -15 -15>, angularTarget=<1 0 0>}
      localJointOffset=<2 0 0>
     }]
   }
   n11 = Anim Dynamics - Bone: dyn_ear_01_l(ComponentPose=@n10, Alpha=1, ExternalForce="0,0,0") {
    angularDampingOverride=0.9998
    angularSpringConstant=5000
    bAngularSpring=true
    bOverrideAngularDamping=true
    bUseGravityOverride=true
    simulationSpace=RootRelative
    alphaBoolBlend {blendInTime=0.1, blendOption=Linear, blendOutTime=0.1}
    boundBone {boneName=dyn_ear_01_l}
    componentAppliedLinearAccClamp=<2500 2500 1000>
    componentLinearAccScale=<0.6 0.6 0.6>
    componentLinearVelScale=<0.8 0.8 0.8>
    gravityOverride {z=-500}
    physicsBodyDefinitions [{
      boundBone {boneName=dyn_ear_01_l}
      boxExtents=<20 20 20>
      constraintSetup {angularLimitsMax=<2.5 50 15>, angularLimitsMin=<-2.5 -30 -25>, angularTarget=<1 0 0>}
      localJointOffset=<2 0 0>
     }]
   }
   n12 = Anim Dynamics - Bone: dyn_ear_02_l(ComponentPose=@n11, Alpha=1, ExternalForce="0,0,0") {
    angularDampingOverride=0.9999
    angularSpringConstant=4500
    bAngularSpring=true
    bOverrideAngularDamping=true
    bUseGravityOverride=true
    simulationSpace=RootRelative
    alphaBoolBlend {blendInTime=0.1, blendOption=Linear, blendOutTime=0.1}
    boundBone {boneName=dyn_ear_02_l}
    componentAppliedLinearAccClamp=<2500 2500 1000>
    componentLinearAccScale=<0.6 0.6 0.6>
    componentLinearVelScale=<0.8 0.8 0.8>
    gravityOverride {z=-500}
    physicsBodyDefinitions [{
      boundBone {boneName=dyn_ear_02_l}
      boxExtents=<15 15 15>
      constraintSetup {angularLimitsMax=<2.5 50 15>, angularLimitsMin=<-2.5 -30 -25>, angularTarget=<1 0 0>}
      localJointOffset=<1 0 0>
     }]
   }
   n13 = Anim Dynamics - Bone: dyn_ear_01_r(ComponentPose=@n12, Alpha=1, ExternalForce="0,0,0") {
    angularDampingOverride=0.9998
    angularSpringConstant=5000
    bAngularSpring=true
    bOverrideAngularDamping=true
    bUseGravityOverride=true
    simulationSpace=RootRelative
    alphaBoolBlend {blendInTime=0.1, blendOption=Linear, blendOutTime=0.1}
    boundBone {boneName=dyn_ear_01_r}
    componentAppliedLinearAccClamp=<2500 2500 1000>
    componentLinearAccScale=<0.6 0.6 0.6>
    componentLinearVelScale=<0.8 0.8 0.8>
    gravityOverride {z=-500}
    physicsBodyDefinitions [{
      boundBone {boneName=dyn_ear_01_r}
      boxExtents=<20 20 20>
      constraintSetup {angularLimitsMax=<2.5 30 15>, angularLimitsMin=<-2.5 -50 -25>, angularTarget=<1 0 0>}
      localJointOffset=<2 0 0>
     }]
   }
   n14 = Anim Dynamics - Bone: dyn_ear_02_r(ComponentPose=@n13, Alpha=1, ExternalForce="0,0,0") {
    angularDampingOverride=0.9999
    angularSpringConstant=4500
    bAngularSpring=true
    bOverrideAngularDamping=true
    bUseGravityOverride=true
    simulationSpace=RootRelative
    alphaBoolBlend {blendInTime=0.1, blendOption=Linear, blendOutTime=0.1}
    boundBone {boneName=dyn_ear_02_r}
    componentAppliedLinearAccClamp=<2500 2500 1000>
    componentLinearAccScale=<0.6 0.6 0.6>
    componentLinearVelScale=<0.8 0.8 0.8>
    gravityOverride {z=-500}
    physicsBodyDefinitions [{
      boundBone {boneName=dyn_ear_02_r}
      boxExtents=<15 15 15>
      constraintSetup {angularLimitsMax=<2.5 30 15>, angularLimitsMin=<-2.5 -50 -25>, angularTarget=<1 0 0>}
      localJointOffset=<1 0 0>
     }]
   }
   n15 = Anim Dynamics - Bone: dyn_ear_03_r(ComponentPose=@n14, Alpha=1, ExternalForce="0,0,0") {
    angularDampingOverride=0.9999
    angularSpringConstant=2500
    bAngularSpring=true
    bOverrideAngularDamping=true
    bUseGravityOverride=true
    simulationSpace=RootRelative
    alphaBoolBlend {blendInTime=0.1, blendOption=Linear, blendOutTime=0.1}
    boundBone {boneName=dyn_ear_03_r}
    componentAppliedLinearAccClamp=<2500 2500 1000>
    componentLinearAccScale=<0.6 0.6 0.6>
    componentLinearVelScale=<0.8 0.8 0.8>
    gravityOverride {z=-500}
    physicsBodyDefinitions [{
      boundBone {boneName=dyn_ear_03_r}
      boxExtents=<10 10 10>
      constraintSetup {angularLimitsMax=<2.5 30 10>, angularLimitsMin=<-2.5 -30 -10>, angularTarget=<1 0 0>}
      localJointOffset=<0.5 0 0>
     }]
   }
   n16 = Anim Dynamics - Bone: dyn_arrow_01(ComponentPose=@n15, Alpha=1, ExternalForce="0,0,0") {
    angularDampingOverride=0.999
    angularSpringConstant=7000
    bAngularSpring=true
    bOverrideAngularDamping=true
    bUseGravityOverride=true
    simulationSpace=RootRelative
    boundBone {boneName=dyn_arrow_01}
    componentAppliedLinearAccClamp=<5000 5000 1000>
    componentLinearAccScale=<0.5 0.5 0.5>
    componentLinearVelScale=<0.7 0.7 0.7>
    gravityOverride {z=-500}
    physicsBodyDefinitions [{
      boundBone {boneName=dyn_arrow_01}
      boxExtents=<25 25 25>
      constraintSetup {angularLimitsMax=<10 10 10>, angularLimitsMin=<-10 -10 -10>, angularTarget=<1 0 0>}
      localJointOffset=<1 0 0>
     }]
   }
   n17 = Anim Dynamics - Bone: dyn_arrow_02(ComponentPose=@n16, Alpha=1, ExternalForce="0,0,0") {
    angularDampingOverride=0.9999
    angularSpringConstant=3000
    bAngularSpring=true
    bOverrideAngularDamping=true
    bUseGravityOverride=true
    simulationSpace=RootRelative
    boundBone {boneName=dyn_arrow_02}
    componentAppliedLinearAccClamp=<1200 1200 1200>
    componentLinearAccScale=<0.5 0.5 0.5>
    componentLinearVelScale=<0.7 0.7 0.7>
    gravityOverride {z=-750}
    physicsBodyDefinitions [{
      boundBone {boneName=dyn_arrow_02}
      boxExtents=<15 15 15>
      constraintSetup {angularLimitsMax=<10 10 10>, angularLimitsMin=<-10 -10 -10>, angularTarget=<1 0 0>}
      localJointOffset=<0.25 0 -1>
     }]
   }
   n18 = Anim Dynamics - Bone: dyn_arrow_03(ComponentPose=@n17, Alpha=1, ExternalForce="0,0,0") {
    angularDampingOverride=0.997
    angularSpringConstant=5500
    bAngularSpring=true
    bOverrideAngularDamping=true
    bUseGravityOverride=true
    simulationSpace=RootRelative
    boundBone {boneName=dyn_arrow_03}
    componentAppliedLinearAccClamp=<2500 2500 5000>
    componentLinearAccScale=<0.5 0.5 0.5>
    componentLinearVelScale=<0.7 0.7 0.7>
    gravityOverride {z=-800}
    physicsBodyDefinitions [{
      boundBone {boneName=dyn_arrow_03}
      boxExtents=<20 20 20>
      constraintSetup {angularLimitsMax=<10 10 10>, angularLimitsMin=<-10 -10 -10>, angularTarget=<1 0 0>}
      localJointOffset=<0.3 -1 0>
     }]
   }
   Output Pose(Result=Blend Poses by bool(BlendPose_0=Component To Local(ComponentPose=@n9), BlendPose_1=Component To Local(ComponentPose=@n18), bActiveValue=Get bIsInFrontEnd())) {layerGroup=None}
   n19 = "Transform (Modify) Bone - Bone: pack"!(ComponentPose=Local To Component(LocalPose=Input Pose()), Alpha, Translation="0,0,0", Rotation="0,0,0") -> (Pose) {
    rotationMode=BMM_Additive
    rotationSpace=BCS_BoneSpace
    translationMode=BMM_Additive
    translationSpace=BCS_BoneSpace
    boneToModify {boneName=pack}
   }
   Save cached pose 'PreSimPose'(Pose=Component To Local(ComponentPose=@n19)) {cacheName=PreSimPose}
  ]}
 EventGraph {nodes [
   Event BlueprintUpdateAnimation()
   n1 = GetOwnerSkeletalMeshComponent()
   n2 = BreakRotator(InRot=BreakTransform(InTransform=GetDeltaTransformFromRefPose(self=@n1, BoneName=spine_05, BaseName=spine_04)).Rotation)
   > Set BackPack Rotation(BackPack Rotation:Rotator=MakeRotator(Roll=MapRangeClamped(Value=@n2.Yaw, InRangeA=0, InRangeB=-30, OutRangeB=30))) -> (Output_Get:Rotator)
   > Set BackPack Translation(BackPack Translation:Vector=MakeVector(Y=MapRangeClamped(Value=@n2.Yaw, InRangeA=0, InRangeB=-30, OutRangeB=-2))) -> (Output_Get:Vector)
  ]}
]
variables [
 BackPack Rotation {type=Rotator}
 BackPack Translation {type=Vector}
]
